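# Translucent, two-sided Niagara mesh-particle material with depth test disabled, apparently a flame overlay for when the camera passes through a spherical volume's boundary.
# A PassThrough value measures camera-to-object distance against Radius, scaled by Threshold.
# A tiling wave-flame texture drives a two-color tint and alpha, alongside panned, swizzled and tilted screen-aligned UV helpers.
# A SmoothStep dissolve and viewport-edge mask are also built, though output connections are not shown.
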Material /Game/Effects/Niagara/Systems/VolumeDone/M_PassThrough
graph {
 properties {
  bAutomaticallySetUsageInEditor=false
  bDisableDepthTest=true
  bUsedWithNiagaraMeshParticles=true
  blendMode=BLEND_Translucent
  twoSided=true
  usedShadingModels=Default Lit
 }
 nodes [
  n1 = Length(None=Subtract(A=CameraPositionWS(), B=ObjectPositionWS())) {is=class MaterialExpressionLength}
  n2 = ScalarParameter() {defaultValue=379.5972, parameterName=Radius}
  Add(A=@n1, B=@n2)
  DynamicParameter() {defaultValue=<rgba 1 1 1 250>, paramNames [Time, DIssolveOffset, Distort, WPO_Width]}
  DynamicParameter() {defaultValue=<rgba 1 1 1 250>, paramNames [Time,Distort,Dissolve,WPO_Width]}
  n3 = TextureSample(UVs=NamedRerouteUsage()) {texture=/Game/Effects/Fort_Effects/Textures/Tile/T_Tile_WaveFlame}
  LinearInterpolate(
   A=VectorParameter() {parameterName=Param_1, defaultValue=<rgba 1 0.2899 0.494 1>}.RGB
   B=VectorParameter() {parameterName=Param_2, defaultValue=<rgba 0.375 0.0839 0.1904 1>}.RGB
   Alpha=SquareRoot(None=@n3.G)
  )
  LinearInterpolate(A=TextureCoordinate(), B=TextureCoordinate() {coordinateIndex=1}, Alpha=ScalarParameter() {parameterName=CubeMap_Switch})
  n4 = TextureCoordinate()
  LinearInterpolate(
   A=@n4
   B=MaterialFunctionCall(XY=@n4) {materialFunction=&Swizzle}.YX
   Alpha=ScalarParameter() {defaultValue=1, parameterName=UV1_Swizzle}
  )
  Multiply(A=OneMinus(None=TwoSidedSign()), B=NamedRerouteUsage())
  n5 = Abs(None=Divide(A=Subtract(A=@n1, B=@n2), B=ScalarParameter() {defaultValue=100, parameterName=Threshold}))
  NamedRerouteDeclaration(None=@n5) {name=PassThrough, nodeColor=<rgba 0 0.1328 1 1>}
  n6 = NamedRerouteUsage()
  n7 = Add(A=ScalarParameter() {defaultValue=-0.24, group=Dissolve, parameterName=DissolveMin}, B=@n6)
  n8 = ScalarParameter() {defaultValue=0.7979, group=Dissolve, parameterName=DissolveMax}
  n9 = NamedRerouteUsage()
  NamedRerouteDeclaration(None=SmoothStep(Min=@n7, Max=Add(A=@n7, B=@n8), Value=@n9)) {name=DissolveResault, nodeColor=<rgba 1 0.9405 0.1679 1>}
  NamedRerouteDeclaration(None=Add(A=@n3.G, B=NamedRerouteUsage())) {name=Alpha, nodeColor=<rgba 1 0.1875 0 1>}
  n10 = VectorParameter() {parameterName=UV1, defaultValue=<rgba 0.028 0.16 -0.3067 0>}
  NamedRerouteDeclaration(None=MaterialFunctionCall!(ClampTexture, UV=NamedRerouteUsage(), Tex, InRGB=@n10.RGB, InA=@n10.RGB, Time=Time(), TimeMult) {materialFunction=/Game/Effects/Niagara/MasterMaterials/Functions/MF_FX_VectorDrive_TexPanner}.PannerUV) {name=UV_1, nodeColor=<rgba 0 0.4049 1 1>}
  n11 = MaterialFunctionCall(Float2=MaterialFunctionCall!(TextureResolution) {materialFunction=/Engine/Functions/Engine_MaterialFunctions02/Texturing/ScreenAlignedPixelToPixelUVs}.Result) {materialFunction=&BreakOutFloat2Components}
  n12 = Multiply(A=@n11.R, B=ScalarParameter() {defaultValue=0.4, parameterName=UV_Tilt})
  NamedRerouteDeclaration(None=MaterialFunctionCall(X=@n11.R, Y=Add(A=@n11.G, B=@n12)) {materialFunction=&MakeFloat2}.Result) {name=TiltedUV, nodeColor=<rgba 0 0.3848 1 1>}
  n13 = Subtract(A=Multiply(A=ScreenPosition().ViewportUV) {constB=2})
  n14 = MaterialFunctionCall(Float2=OneMinus(None=Abs(None=@n13))) {materialFunction=&BreakOutFloat2Components}
  NamedRerouteDeclaration(None=OneMinus(None=Multiply(A=@n14.R, B=@n14.R))) {name=CubeMap_2, nodeColor=<rgba 1 1 1 1>}
  OneMinus(None=Multiply(A=@n9, B=@n6))
  Saturate(None=OneMinus(None=@n5))
  TextureCoordinate()
 ]
}
