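# Masked, two-sided Subsurface material for Niagara mesh particles in the VolumeDone effect, despite its Opaque name.
# Named reroutes define panned, tilted and screen-aligned UVs, an invertible Fresnel, and a camera-distance PassThrough band.
# Separate branches build a two-color flame-texture blend and a ray-traced sphere falloff mask with depth fade.
# No connection to a material output appears, and several branches and helper nodes end unused, suggesting an unfinished or stripped-down graph.
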Material /Game/Effects/Niagara/Systems/VolumeDone/M_VolumeDone_Opaque
graph {
 properties {
  bAutomaticallySetUsageInEditor=false
  bUsedWithNiagaraMeshParticles=true
  blendMode=BLEND_Masked
  shadingModel=MSM_Subsurface
  twoSided=true
  usedShadingModels=Subsurface
 }
 nodes [
  n1 = Length(None=Subtract(A=CameraPositionWS(), B=ObjectPositionWS())) {is=class MaterialExpressionLength}
  n2 = ScalarParameter() {defaultValue=300.8903, parameterName=Radius}
  Add(A=@n1, B=@n2)
  n3 = ObjectPositionWS()
  Clamp(None=TransformPosition(None=@n3) {transformSourceType=TRANSFORMPOSSOURCE_World, transformType=TRANSFORMPOSSOURCE_Camera})
  Constant()
  DynamicParameter() {defaultValue=<rgba 1 1 1 250>, paramNames [Time,Distort,Dissolve,WPO_Width]}
  n4 = NamedRerouteUsage()
  LinearInterpolate(
   A=VectorParameter() {parameterName=Param_1, defaultValue=<rgba 0.8733 0.3654 0.5018 1>}.RGB
   B=VectorParameter() {parameterName=Param_2, defaultValue=<rgba 0.3802 0.0198 0.1876 1>}.RGB
   Alpha=SquareRoot(None=TextureSample(UVs=@n4) {texture=/Game/Effects/Fort_Effects/Textures/Tile/T_Tile_WaveFlame}.G)
  )
  n5 = TwoSidedSign()
  LinearInterpolate(A=@n5, B=Subtract(B=@n5), Alpha=ScalarParameter() {parameterName=Side})
  n6 = TextureCoordinate()
  LinearInterpolate(
   A=@n6
   B=MaterialFunctionCall(XY=@n6) {materialFunction=&Swizzle}.YX
   Alpha=ScalarParameter() {defaultValue=1, parameterName=Main_UV_Swizzle}
  )
  MaterialFunctionCall(Fade Length=ScalarParameter() {defaultValue=50, parameterName=Param_6}) {materialFunction=&CameraDepthFade}
  MaterialFunctionCall() {materialFunction=&WorldPositionBehindTranslucency}
  MaterialFunctionCall(Offset Amount=ScalarParameter() {parameterName=Param_4}) {materialFunction=/Engine/Functions/Engine_MaterialFunctions02/WorldPositionOffset/CameraOffset}
  n7 = MaterialFunctionCall!(
   Sphere Radius
   Sphere Falloff=ScalarParameter() {defaultValue=438.0609, parameterName=Param_1}
   Unit Ray Direction
   Sphere Center=@n3
   Ray Origin
  ) {materialFunction=/Engine/Functions/Engine_MaterialFunctions02/Gradients/RayTraceSphereFalloff}
  Multiply(
   A=SmoothStep(
    Min=ScalarParameter() {parameterName=Param_2}
    Max=ScalarParameter() {defaultValue=0.1618, parameterName=Param_3}
    Value=OneMinus(None=@n7.Clamped Falloff)
   )
   B=DepthFade(FadeDistance=ScalarParameter() {defaultValue=50, group=Opacity, parameterName=DepthFade})
  )
  n8 = Abs(None=Divide(A=Subtract(A=@n1, B=@n2), B=ScalarParameter() {defaultValue=50, parameterName=Threshold}))
  NamedRerouteDeclaration(None=Saturate(None=OneMinus(None=@n8))) {name=PassThrough, nodeColor=<rgba 0 0.1328 1 1>}
  n9 = VectorParameter() {parameterName=Main_UV, defaultValue=<rgba 2 2 -0.3067 0>}
  NamedRerouteDeclaration(None=MaterialFunctionCall!(ClampTexture, UV=NamedRerouteUsage(), Tex, InRGB=@n9.RGB, InA=@n9.RGB, Time, TimeMult) {materialFunction=/Game/Effects/Niagara/MasterMaterials/Functions/MF_FX_VectorDrive_TexPanner}.PannerUV) {name=MainUV, nodeColor=<rgba 0 0.4049 1 1>}
  n10 = Subtract(A=WorldPosition().XYZ, B=ObjectPositionWS())
  n11 = ComponentMask(None=Divide(
    A=Transform(None=@n10) {transformSourceType=TRANSFORMSOURCE_Local, transformType=TRANSFORM_Camera}
    B=ScalarParameter() {defaultValue=196.4626, group=ScreenAlign UV, parameterName=ScreenAlign UV Scale}
   )) {g=true, r=true}
  NamedRerouteDeclaration(None=Saturate(None=Add(A=@n11) {constB=0.5})) {name=ScreenAlignUV, nodeColor=<rgba 0 0.3848 1 1>}
  n12 = CameraPositionWS()
  n13 = WorldPosition()
  n14 = DotProduct(A=VertexNormalWS(), B=Normalize(VectorInput=Subtract(A=@n12, B=@n13.XYZ)))
  n15 = Abs(None=@n14)
  n16 = Power(
   Base=LinearInterpolate(A=OneMinus(None=@n15), B=@n15, Alpha=ScalarParameter() {defaultValue=1, parameterName="Fresnel_Invert (0or1)"})
   Exp=ScalarParameter() {defaultValue=2, parameterName=Fresnel_Power}
  )
  n17 = ScalarParameter() {parameterName=Fresnel_Offset}
  NamedRerouteDeclaration(None=Multiply(A=Subtract(A=@n16, B=@n17), B=Add(A=@n17))) {name=Fresnel, nodeColor=<rgba 1 0 0.9613 1>}
  n18 = TextureCoordinate()
  n19 = MaterialFunctionCall(XY=@n18) {materialFunction=&Swizzle}
  n20 = ScalarParameter() {defaultValue=1, parameterName=UV1_Swizzle}
  n21 = VectorParameter() {parameterName=UV1, defaultValue=<rgba 0.2453 0.2582 -0.3067 0>}
  NamedRerouteDeclaration(None=MaterialFunctionCall(
    UV=LinearInterpolate(A=@n18, B=@n19.YX, Alpha=@n20)
    InRGB=@n21.RGB
    InA=@n21.RGB
    Time=DynamicParameter() {defaultValue=<rgba 1 1 1 250>, paramNames [Time,Distort,Dissolve,WPO_Width]}.Time
   ) {materialFunction=&MF_FX_VectorDrive_TexPanner}.PannerUV) {name=UV_1, nodeColor=<rgba 0 0.4049 1 1>}
  n22 = MaterialFunctionCall(Float2=TextureCoordinate()) {materialFunction=&BreakOutFloat2Components}
  n23 = Multiply(A=@n22.R, B=ScalarParameter() {defaultValue=1, parameterName=UV_Tilt})
  NamedRerouteDeclaration(None=MaterialFunctionCall(X=@n22.R, Y=Add(A=@n22.G, B=@n23)) {materialFunction=&MakeFloat2}.Result) {name=TiltedUV, nodeColor=<rgba 0 0.3848 1 1>}
  Reroute(None=@n14)
  ScalarParameter() {defaultValue=1, parameterName=Param_5}
  VectorParameter() {defaultValue=<rgba 0.0955 0.0336 0.061 1>}
  WorldPosition()
 ]
}
